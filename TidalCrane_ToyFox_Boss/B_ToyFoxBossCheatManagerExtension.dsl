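; Cheat manager extension that adds debug commands to force Toy Fox Queen boss actions in the TidalCrane boss fight.
; Each command validates cheat data, then looks up the boss through GetToyFoxQueen.
; Stun and takeover call the boss's BPI interface directly.
; Melee, ranged attack and reposition get its Ability System Component through that interface and try to activate abilities by gameplay tag.
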(fn ToyFoxQueenApplyStun
  (Default|ValidateCheatData
    (:Valid
      (bind _boss (Utilities|GetToyFoxQueen)
        (:Valid
          (Class|BPITidalCraneToyFoxBoss|ToyFoxQueenApplyStun _boss))))))

(fn ToyFoxQueenMeleeAttack
  (Default|ValidateCheatData
    (:Valid
      (bind _boss (Utilities|GetToyFoxQueen)
        (:Valid
          (bind _component (Class|BPITidalCraneToyFoxBoss|GetAbilitySystemComponent _boss))
          (Utilities|IsValid _component
            (:Is%20Valid
              (Abilities|TryActivateAbilitiesbyTag _component (GameplayTags|MakeLiteralGameplayTagContainer "(GameplayTags=((TagName=\"Ability.TidalCrane.Boss.Melee\")))") true))))))))

(fn ToyFoxQueenRangeAttack
  (Default|ValidateCheatData
    (:Valid
      (bind _boss (Utilities|GetToyFoxQueen)
        (:Valid
          (bind _component (Class|BPITidalCraneToyFoxBoss|GetAbilitySystemComponent _boss))
          (Utilities|IsValid _component
            (:Is%20Valid
              (Abilities|TryActivateAbilitiesbyTag _component (GameplayTags|MakeLiteralGameplayTagContainer "(GameplayTags=((TagName=\"Ability.TidalCrane.Boss.RangeAttack\")))") true))))))))

(fn ToyFoxQueenReposition
  (Default|ValidateCheatData
    (:Valid
      (bind _boss (Utilities|GetToyFoxQueen)
        (:Valid
          (bind _component (Class|BPITidalCraneToyFoxBoss|GetAbilitySystemComponent _boss))
          (Utilities|IsValid _component
            (:Is%20Valid
              (Abilities|TryActivateAbilitiesbyTag _component (GameplayTags|MakeLiteralGameplayTagContainer "(GameplayTags=((TagName=\"Ability.TidalCrane.Boss.Reposition\")))") true))))))))

(fn ToyFoxQueenStartTakeOver
  (Default|ValidateCheatData
    (:Valid
      (bind _boss (Utilities|GetToyFoxQueen)
        (:Valid
          (Class|BPITidalCraneToyFoxBoss|ToyFoxQueenStartTakeOver _boss))))))
