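; Blueprint static point provider that spawns the Toy Fox boss through the Fortnite Living World system.
; Two custom events, SpawnToyFoxBoss and EnableProvider, run identical logic: each enables this point provider via a deprecated interface call, then asks the Living World Manager to request event generation.
; The construction script is empty, and no spawn points or boss data appear in the graph.
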(event Custom|SpawnToyFoxBoss
  (Class|FortAthenaLivingWorldPointProviderInterface|EnablePointProvider_DEPRECATED self)
  (Class|FortAthenaLivingWorldManager|RequestEventGeneration (Class|FortAthenaLivingWorldManager|GetLivingWorldManager)))

(event Custom|EnableProvider
  (Class|FortAthenaLivingWorldPointProviderInterface|EnablePointProvider_DEPRECATED self)
  (Class|FortAthenaLivingWorldManager|RequestEventGeneration (Class|FortAthenaLivingWorldManager|GetLivingWorldManager)))

(fn ConstructionScript)
